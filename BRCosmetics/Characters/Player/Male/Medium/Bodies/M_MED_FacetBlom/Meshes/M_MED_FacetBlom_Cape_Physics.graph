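# Physics asset for the M_MED_FacetBlom male cape mesh, covering a dangling belt piece on the character.
# Three capsule bodies form a dyn_belt chain constrained back to a tiny MSTR_belt capsule.
# Tight swing and twist limits keep the chain stiff, with slightly more swing at the last link.
# Pelvis and left thigh capsules, two on the thigh, have no constraints and act as body collision volumes for the chain.
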PhysicsAsset /BRCosmetics/Characters/Player/Male/Medium/Bodies/M_MED_FacetBlom/Meshes/M_MED_FacetBlom_Cape_Physics
bodies {
 MSTR_belt [{shapeType=Capsule, length=0.505, radius=0.505, shapeName=MSTR_belt_capsule, rotation=<rot 0 -90 0>}]
 dyn_belt_01 [{
   shapeType=Capsule
   length=7.3793
   radius=1.4762
   shapeName=dyn_belt_01_capsule
   center=<5.1434 0.0848 -0.1425>
   rotation=<rot 0 90.7169 -88.7317>
  }]
 dyn_belt_02 [{
   shapeType=Capsule
   length=7.8997
   radius=1.8143
   shapeName=dyn_belt_02_capsule
   center=<5.4576 -0.012 0.1044>
   rotation=<rot 0.0398 89.8951 -88.7599>
  }]
 dyn_belt_03 [{
   shapeType=Capsule
   length=5.906
   radius=1.787
   shapeName=dyn_belt_03_capsule
   center=<4.8972 -0.0472 0.1131>
   rotation=<rot 0 89.5101 -88.7344>
  }]
 pelvis [{shapeType=Capsule, length=15.1917, radius=9.9268, shapeName=pelvis_capsule, center=<-1.4161 -2.367 0>}]
 thigh_l [
  {
   shapeType=Capsule
   length=23.7853
   radius=11.2684
   shapeName=thigh_l_capsule
   center=<-14.372 -0.6813 0>
   rotation=<rot 85.6581 -176.9965 179.9999>
  }
  {
   shapeType=Capsule
   length=5.8137
   radius=10.9073
   shapeName=thigh_l_capsule
   center=<-17.2675 -2.1848 0.1848>
   rotation=<rot 89.3868 0.0019 -4.8318>
  }
 ]
}
constraints [
 {bone1Name=dyn_belt_01, bone2Name=MSTR_belt, swing1LimitDegrees=5, swing2LimitDegrees=5, twistLimitDegrees=2}
 {bone1Name=dyn_belt_02, bone2Name=dyn_belt_01, swing1LimitDegrees=5, swing2LimitDegrees=5, twistLimitDegrees=2}
 {bone1Name=dyn_belt_03, bone2Name=dyn_belt_02, swing1LimitDegrees=8, swing2LimitDegrees=8, twistLimitDegrees=2}
]
