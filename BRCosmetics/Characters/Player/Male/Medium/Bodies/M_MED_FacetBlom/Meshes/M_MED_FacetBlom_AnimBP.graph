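# Post-process style Anim Blueprint for the FacetBlom medium male cosmetic body, copying the parent mesh pose into a Control Rig.
# Bone offsets and a clasp constraint feed two RigidBody sims, with a tighter-clamped cape variant selected when long or cape backbling is worn.
# AnimDynamics springs clasps, zipper and collar, then backpack, belt and clasp offsets apply.
# A front-end idle curve blends DefaultSlot onto sword bones; EventGraph updates gravity and backpack state.
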AnimBlueprint /BRCosmetics/Characters/Player/Male/Medium/Bodies/M_MED_FacetBlom/Meshes/M_MED_FacetBlom_AnimBP
target_skeleton=&/M_MED_FacetBlom_Skeleton
graphs [
 AnimGraph {nodes [
   n1 = "Transform (Modify) Bone - Bone: attach_backpack"(
    ComponentPose=Local To Component(LocalPose=Use cached pose 'After dynamics'())
    Alpha=1
    Translation="-10.102,2.419,0.000"
    Rotation="-2.159,6.019,-1.285"
    Scale="1,1,1"
   ) {
    rotationMode=BMM_Additive
    translationMode=BMM_Additive
    alphaScaleBiasClamp {bInterpResult=true, interpSpeedDecreasing=6}
    boneToModify {boneName=attach_backpack}
    rotation=<rot -2.159 6.019 -1.285>
    translation=<-10.102 2.419>
   }
   n2 = "Transform (Modify) Bone - Bone: attach_backpack"(ComponentPose=@n1, Alpha=1, Translation="0,0,0", Rotation="0.000,23.906,-0.000", Scale="1,1,1") {
    rotationMode=BMM_Additive
    translationMode=BMM_Additive
    alphaScaleBiasClamp {bInterpResult=true, interpSpeedDecreasing=6}
    boneToModify {boneName=attach_backpack}
    rotation {yaw=23.906}
   }
   n3 = "Transform (Modify) Bone - Bone: MSTR_belt"(ComponentPose=@n2, Alpha=1, Translation="0,0,0", Rotation="0,10,20", Scale="1,1,1") {
    rotationMode=BMM_Additive
    translationMode=BMM_Additive
    alphaScaleBiasClamp {bInterpResult=true, interpSpeedDecreasing=6}
    boneToModify {boneName=MSTR_belt}
    rotation {roll=20, yaw=10}
   }
   n4 = "Transform (Modify) Bone - Bone: dyn_clasp_01"(ComponentPose=@n3, Alpha=1, Translation="0,0,0", Rotation="0,0,15", Scale="1,1,1") {
    rotationMode=BMM_Additive
    translationMode=BMM_Additive
    alphaScaleBiasClamp {bInterpResult=true, interpSpeedDecreasing=6}
    boneToModify {boneName=dyn_clasp_01}
    rotation {roll=15}
   }
   Output Pose(Result=Layered blend per bone(
     BasePose=Component To Local(ComponentPose=@n4)
     BlendPoses_0=Slot 'DefaultSlot'()
     BlendWeights_0=Get CustomIdleIsPlayingFloat() -> (CustomIdleIsPlayingFloat:double)
    ) {layerSetup [{branchFilters [{boneName=MSTR_sword}, {boneName=MSTR_sword_Direct}]}]}) {layerGroup=None}
   n5 = Control Rig(Source=Copy Pose From Mesh() {bUseAttachedParent=true}) {controlRigAssetReference {blueprintRigClass=&/M_MED_FacetBlom_CtrlRig.M_MED_FacetBlom_CtrlRig_C}}
   Save cached pose 'control_rig'(Pose=@n5) {cacheName=control_rig}
   n6 = "Transform (Modify) Bone - Bone: MSTR_sword"(
    ComponentPose=Local To Component(LocalPose=Use cached pose 'control_rig'())
    Alpha=1
    Translation="-7.476,-2.676,-0.000"
    Rotation="27.902,-22.716,6.971"
    Scale="1,1,1"
   ) {
    rotationMode=BMM_Additive
    translationMode=BMM_Additive
    alphaScaleBiasClamp {bInterpResult=true, interpSpeedDecreasing=20, interpSpeedIncreasing=5}
    boneToModify {boneName=MSTR_sword}
    rotation=<rot 27.902 -22.716 6.971>
    translation=<-7.476 -2.676>
   }
   n7 = "Transform (Modify) Bone - Bone: MSTR_sword"(ComponentPose=@n6, Alpha=1, Translation="0.587,-1.113,2.241", Rotation="-24.905,19.168,30.118", Scale="1,1,1") {
    rotationMode=BMM_Additive
    translationMode=BMM_Additive
    alphaScaleBiasClamp {interpSpeedIncreasing=100}
    boneToModify {boneName=MSTR_sword}
    rotation=<rot -24.905 19.168 30.118>
    translation=<0.587 -1.113 2.241>
   }
   n8 = "Transform (Modify) Bone - Bone: MSTR_sword"(ComponentPose=@n7, Alpha=1, Translation="0,0,0", Rotation="-10,0,0", Scale="1,1,1") {
    rotationMode=BMM_Additive
    translationMode=BMM_Additive
    alphaScaleBiasClamp {bInterpResult=true, interpSpeedDecreasing=6}
    boneToModify {boneName=MSTR_sword}
    rotation {pitch=-10}
   }
   n9 = "Constraint - MSTR_clasp (PPP)"!(ComponentPose=@n8, Alpha=1, ConstraintWeights_0=0.4903, ConstraintWeights_1:float=0.4097, ConstraintWeights_2:float=0.1001) -> (Pose) {
    boneToModify {boneName=MSTR_clasp}
    constraintSetup [{targetBone {boneName=spine_05}}, {targetBone {boneName=spine_04}}, {targetBone {boneName=spine_03}}]
    constraintWeights [1,1,1]
   }
   Save cached pose 'before dynamics'(Pose=Component To Local(ComponentPose=@n9)) {cacheName=before dynamics}
   n10 = RigidBody!(
    ComponentPose=Local To Component(LocalPose=Use cached pose 'before dynamics'())
    Alpha
    OverrideWorldGravity:Vector="0,0,-980"
    ExternalForce
   ) -> (Pose) {
    bOverrideWorldGravity=true
    overridePhysicsAsset=&/M_MED_FacetBlom_Physics
    baseBoneRef {boneName=pelvis}
    componentAppliedLinearAccClamp=<3000 3000 3000>
    overrideWorldGravity {z=-980}
    simSpaceSettings {maxAngularAcceleration=5500, maxAngularVelocity=100, maxLinearAcceleration=6500, maxLinearVelocity=100, velocityScaleZ=0.5, worldAlpha=1}
   }
   n11 = Component To Local(ComponentPose=@n10)
   n12 = RigidBody(ComponentPose=Local To Component(LocalPose=Use cached pose 'before dynamics'()), OverrideWorldGravity="0,0,-980") {
    bOverrideWorldGravity=true
    overridePhysicsAsset=&/M_MED_FacetBlom_Cape_Physics
    baseBoneRef {boneName=pelvis}
    componentAppliedLinearAccClamp=<3000 3000 3000>
    overrideWorldGravity {z=-980}
    simSpaceSettings {maxAngularAcceleration=40, maxAngularVelocity=8, maxLinearAcceleration=60, maxLinearVelocity=15, velocityScaleZ=0.5, worldAlpha=1}
   }
   n13 = Component To Local(ComponentPose=@n12)
   n14 = Get IsWearingCape() -> (IsWearingCape:bool)
   n15 = Anim Dynamics - Bone: dyn_clasp_01(
    ComponentPose=Local To Component(LocalPose=Blend Poses by bool(BlendPose_0=@n13, BlendPose_1=@n11, bActiveValue=@n14) {blendType=Linear})
    Alpha=1
    ExternalForce="0,0,0"
   ) {
    angularDampingOverride=1
    angularSpringConstant=6500
    bAngularSpring=true
    bOverrideAngularDamping=true
    bUseGravityOverride=true
    gravityScale=0.35
    simulationSpace=RootRelative
    boundBone {boneName=dyn_clasp_01}
    componentLinearAccScale=<1 1 1>
    componentLinearVelScale=<1 1 1>
    gravityOverride {z=-980}
    physicsBodyDefinitions [{
      boundBone {boneName=dyn_clasp_01}
      boxExtents=<20 20 30>
      constraintSetup {angularLimitsMax=<15 15 35>, angularLimitsMin=<-15 -15 -25>, angularTarget=<1 0 0>}
      localJointOffset=<1 -0.5 0>
     }]
    simSpaceSettings {maxAngularAcceleration=35, maxAngularVelocity=6, simSpaceAngularAlpha=1}
   }
   n16 = Anim Dynamics - Bone: dyn_clasp_02(ComponentPose=@n15, Alpha=1, ExternalForce="0,0,0") {
    angularDampingOverride=1
    angularSpringConstant=7500
    bAngularSpring=true
    bOverrideAngularDamping=true
    bUseGravityOverride=true
    gravityScale=0.35
    simulationSpace=RootRelative
    boundBone {boneName=dyn_clasp_02}
    componentLinearAccScale=<1 1 1>
    componentLinearVelScale=<1 1 1>
    gravityOverride {z=-980}
    physicsBodyDefinitions [{
      boundBone {boneName=dyn_clasp_02}
      boxExtents=<20 30 30>
      constraintSetup {angularLimitsMax=<15 15 15>, angularLimitsMin=<-15 -15 -15>, angularTarget=<1 0 0>}
      localJointOffset=<0.5 -1 0>
     }]
    simSpaceSettings {maxAngularAcceleration=40, maxAngularVelocity=8, simSpaceAngularAlpha=1}
   }
   n17 = Anim Dynamics - Bone: dyn_sword_zipper(ComponentPose=@n16, Alpha=1, ExternalForce="0,0,0") {
    angularDampingOverride=0.99
    angularSpringConstant=100
    bAngularSpring=true
    bOverrideAngularDamping=true
    bUseGravityOverride=true
    gravityScale=0.35
    simulationSpace=RootRelative
    boundBone {boneName=dyn_sword_zipper}
    componentLinearAccScale=<1 1 1>
    componentLinearVelScale=<1 1 1>
    gravityOverride {z=-980}
    physicsBodyDefinitions [{
      boundBone {boneName=dyn_sword_zipper}
      boxExtents=<18 20 20>
      constraintSetup {angularLimitsMax=<1 110 4>, angularLimitsMin=<-1 -3 -4>, angularTarget=<1 0 0>}
      localJointOffset=<3.5 0 0>
     }]
    simSpaceSettings {maxAngularAcceleration=64, maxAngularVelocity=8, simSpaceAngularAlpha=1, externalAngularVelocity=<0.1 0.1 0.5>}
   }
   n18 = Anim Dynamics - Bone: dyn_collar_l(ComponentPose=@n17, Alpha=1, ExternalForce="0,0,0") {
    angularDampingOverride=0.99
    angularSpringConstant=10000
    bAngularSpring=true
    bOverrideAngularDamping=true
    bUseGravityOverride=true
    gravityScale=0.35
    simulationSpace=RootRelative
    boundBone {boneName=dyn_collar_l}
    componentLinearAccScale=<1 1 1>
    componentLinearVelScale=<1 1 1>
    gravityOverride {z=-980}
    physicsBodyDefinitions [{
      boundBone {boneName=dyn_collar_l}
      boxExtents=<20 20 20>
      constraintSetup {angularLimitsMax=<15 15 15>, angularLimitsMin=<-15 -15 -15>, angularTarget=<1 0 0>}
      localJointOffset=<4 0 0>
     }]
   }
   n19 = Anim Dynamics - Bone: dyn_collar_r(ComponentPose=@n18, Alpha=1, ExternalForce="0,0,0") {
    angularDampingOverride=0.99
    angularSpringConstant=10000
    bAngularSpring=true
    bOverrideAngularDamping=true
    bUseGravityOverride=true
    gravityScale=0.35
    simulationSpace=RootRelative
    boundBone {boneName=dyn_collar_r}
    componentLinearAccScale=<1 1 1>
    componentLinearVelScale=<1 1 1>
    gravityOverride {z=-980}
    physicsBodyDefinitions [{
      boundBone {boneName=dyn_collar_r}
      boxExtents=<20 20 20>
      constraintSetup {angularLimitsMax=<15 15 15>, angularLimitsMin=<-15 -15 -15>, angularTarget=<1 0 0>}
      localJointOffset=<4 0 0>
     }]
   }
   n20 = Anim Dynamics - Bone: dyn_collar_c(ComponentPose=@n19, Alpha=1, ExternalForce="0,0,0") {
    angularDampingOverride=0.99
    angularSpringConstant=10000
    bAngularSpring=true
    bOverrideAngularDamping=true
    bUseGravityOverride=true
    gravityScale=0.35
    simulationSpace=RootRelative
    boundBone {boneName=dyn_collar_c}
    componentLinearAccScale=<1 1 1>
    componentLinearVelScale=<1 1 1>
    gravityOverride {z=-980}
    physicsBodyDefinitions [{
      boundBone {boneName=dyn_collar_c}
      boxExtents=<20 20 20>
      constraintSetup {angularLimitsMax=<15 15 15>, angularLimitsMin=<-15 -15 -15>, angularTarget=<1 0 0>}
      localJointOffset=<3 0 0>
     }]
   }
   Save cached pose 'After dynamics'(Pose=Component To Local(ComponentPose=@n20)) {cacheName=After dynamics}
  ]}
 EventGraph {nodes [
   Event BlueprintUpdateAnimation()
   > n1 = Sequence!(execute) -> (then_0, then_1, then_2:Exec, then_3:Exec)
   > n2 = "Gravity Override State Function (Parameter Asset) (Nativized)"(
    CharacterPartAnimInstance=Self-Reference()
    GameTimeInSeconds=GetGameTimeInSeconds()
    Parameters=&/M_MED_FacetBlom_Outfit_Parameter
   )
   > Set GravityOverride(GravityOverride=@n2)
   Branch(execute=@n1.then_2, Condition=Get bIsInFrontEnd())
   > Set CustomIdleIsPlayingFloat(CustomIdleIsPlayingFloat:double=GetCurveValue(self=GetOwnerFrontEndAnimInstance(), CurveName=Frontend_Idle_Bypass)) -> (Output_Get:double)
   n3 = Get OwnBackpackEquiped() -> (OwnBackpackEquiped:bool)
   Set ParachuteOpenedEquipedOwnBackpack(execute=@n1.then_3, ParachuteOpenedEquipedOwnBackpack:bool=AND Boolean(A=@n3, B=Get bIsParachuteOpened())) -> (Output_Get:bool)
   > Set SkyDiveWithOwnEqupedBackpack(SkyDiveWithOwnEqupedBackpack:bool=AND Boolean(A=@n3, B=Get bIsSkydiveFloating())) -> (Output_Get:bool)
   n4 = CheckForBackbling()
   Set IsWearingCape(execute=@n1.then_1, IsWearingCape:bool=OR Boolean(A=@n4.bIsBackblingACape, B=@n4.bIsBackblingLong)) -> (Output_Get:bool)
   Event OnCharacterCustomizationLoaded()
   > Set OwnBackpackEquiped(OwnBackpackEquiped:bool=BackpackItemDefinitionHasAnyMetaTag!(Tags:Array of Gameplay Tag Structures=Get Tags() -> (Tags:Array of Gameplay Tag Structures), bRequiresEquipped:bool=false) -> (ReturnValue:bool)) -> (Output_Get:bool)
  ]}
]
variables [
 CustomIdleIsPlayingFloat {type=number}
 GravityOverride {type=Vector}
 IsWearingCape {type=boolean}
 OwnBackpackEquiped {type=boolean}
 ParachuteOpenedEquipedOwnBackpack {type=boolean}
 SkyDiveWithOwnEqupedBackpack {type=boolean}
 Tags {type="array<GameplayTag>", default [{tagName=Cosmetics.MetaTags.FacetBlom}]}
]
