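# Physics asset for the BackpackEquip face accessory mesh of the male FacetBlom character, covering a dangling rag and the body around it.
# Six dyn_rag capsule bodies form a chain from MSTR_rag, each joint having moderate swing limits and almost no twist.
# Unconstrained capsules on the torso, head, neck, arms, collar and sword provide collision volumes.
# Upper spine bones stack several capsules, including a long angled one offset on spine_05.
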PhysicsAsset /BRCosmetics/Characters/Player/Male/Medium/Bodies/M_MED_FacetBlom/Meshes/Parts/M_MED_FacetBlom_FaceAcc_BackpackEquip_Physics
bodies {
 MSTR_collar_c [{
   shapeType=Capsule
   length=7.5695
   radius=4.25
   shapeName=MSTR_collar_c_capsule
   center=<0.6732 6.0461 -12.0458>
   rotation=<rot 0 0.0001 -70.7167>
  }]
 MSTR_rag [{shapeType=Capsule, length=0.505, radius=0.505, shapeName=MSTR_rag_capsule, rotation=<rot 0 -90 0>}]
 MSTR_sword [{
   shapeType=Capsule
   length=60
   radius=6
   shapeName=MSTR_sword_capsule
   center=<-1.4684 0.9552 -0.2477>
   rotation=<rot -2.0688 -100.261 78.7226>
  }]
 clavicle_l [{shapeType=Capsule, length=10, radius=11.2672, shapeName=clavicle_l_capsule, center=<8 0 0.9975>, rotation=<rot 89 0 0>}]
 clavicle_r [{shapeType=Capsule, length=10, radius=11.2672, shapeName=clavicle_r_capsule, center=<-8 0 -0.9975>, rotation=<rot 89 0 -180>}]
 dyn_rag_01 [{shapeType=Capsule, length=12, radius=3.5, shapeName=dyn_rag_01_capsule, center=<8 0 -3>, rotation=<rot 90 0 0>}]
 dyn_rag_02 [{shapeType=Capsule, length=10, radius=8, shapeName=dyn_rag_02_capsule, center=<8 0 -5>, rotation=<rot 90 130 0>}]
 dyn_rag_03 [{shapeType=Capsule, length=10, radius=9, shapeName=dyn_rag_03_capsule, center=<8 0 -5>, rotation=<rot 0.0002 20 90.0001>}]
 dyn_rag_04 [{shapeType=Capsule, length=6, radius=10, shapeName=dyn_rag_04_capsule, center=<8 -1.9426 -5>, rotation=<rot 0 -0.509 -87.0514>}]
 dyn_rag_05 [{shapeType=Capsule, length=8.8765, radius=8, shapeName=dyn_rag_05_capsule, center=<9.5 -5.8774 -5>, rotation=<rot 0.0001 80 90.0004>}]
 dyn_rag_06 [{
   shapeType=Capsule
   length=15.2052
   radius=2.8686
   shapeName=dyn_rag_06_capsule
   center=<11.8311 -10.9729 0>
   rotation=<rot 0 73.479 -88.7665>
  }]
 head [{shapeType=Capsule, length=8, radius=10, shapeName=head_capsule, center=<8.3024 1.7057 -0.0037>, rotation=<rot -0.0615 17.9709 -89.831>}]
 lowerarm_l [{shapeType=Capsule, length=18.25, radius=6.6023, shapeName=lowerarm_l_capsule, center=<8.0201 0 0>, rotation=<rot 90 0 0>}]
 lowerarm_r [{shapeType=Capsule, length=19.937, radius=9.2578, shapeName=lowerarm_r_capsule, center=<-8.0201 0 0>, rotation=<rot 90 180 0>}]
 neck_01 [{shapeType=Capsule, length=0.505, radius=0.505, shapeName=neck_01_capsule, rotation=<rot 0 -90 0>}]
 neck_02 [{shapeType=Capsule, length=6, radius=6, shapeName=neck_02_capsule, center=<0 -2.9156 0>, rotation=<rot 0 90 90>}]
 pelvis [{shapeType=Capsule, length=35, radius=15, shapeName=pelvis_capsule, center=<7.7829 -7.4576 -0.2602>}]
 spine_01 [{shapeType=Capsule, length=0.505, radius=0.505, shapeName=spine_01_capsule, rotation=<rot 0 -90 0>}]
 spine_02 [{shapeType=Capsule, length=0.505, radius=0.505, shapeName=spine_02_capsule, rotation=<rot 0 -90 0>}]
 spine_03 [{shapeType=Capsule, length=35, radius=15, shapeName=spine_03_capsule, center=<0 -4.7957 0>}]
 spine_04 [
  {shapeType=Capsule, length=30, radius=15, shapeName=spine_05_capsule, center=<0 -4.0806 0>}
  {shapeType=Capsule, length=15.8806, radius=14.1194, shapeName=spine_05_capsule, center=<5.8966 -3.3806 0>, rotation=<rot 0 5.2032 0>}
 ]
 spine_05 [
  {shapeType=Capsule, length=30, radius=15, shapeName=spine_05_capsule, center=<0 -1.5986 0>}
  {shapeType=Capsule, length=15.8806, radius=12, shapeName=spine_05_capsule, center=<9.5643 -6.8929 0>, rotation=<rot 0 5.2032 0>}
  {
   shapeType=Capsule
   length=111.1971
   radius=6
   shapeName=spine_05_capsule
   center=<-10.0282 -22.7194 -3.1839>
   rotation=<rot -56.9148 -2.1179 8.7956>
  }
 ]
 upperarm_l [{shapeType=Capsule, length=19.9972, radius=8.7012, shapeName=upperarm_l_capsule, center=<7.8495 0 0>, rotation=<rot 90 0 0>}]
 upperarm_r [{shapeType=Capsule, length=19.9972, radius=8.7012, shapeName=upperarm_r_capsule, center=<-7.8495 0 0>, rotation=<rot 90 180 0>}]
}
constraints [
 {bone1Name=dyn_rag_01, bone2Name=MSTR_rag, swing1LimitDegrees=25, swing2LimitDegrees=25, twistLimitDegrees=1}
 {bone1Name=dyn_rag_02, bone2Name=dyn_rag_01, swing1LimitDegrees=25, swing2LimitDegrees=25, twistLimitDegrees=1}
 {bone1Name=dyn_rag_03, bone2Name=dyn_rag_02, swing1LimitDegrees=25, swing2LimitDegrees=25, twistLimitDegrees=1}
 {bone1Name=dyn_rag_04, bone2Name=dyn_rag_03, swing1LimitDegrees=25, swing2LimitDegrees=25, twistLimitDegrees=1}
 {bone1Name=dyn_rag_05, bone2Name=dyn_rag_04, swing1LimitDegrees=25, swing2LimitDegrees=25, twistLimitDegrees=1}
 {bone1Name=dyn_rag_06, bone2Name=dyn_rag_05, swing1LimitDegrees=25, swing2LimitDegrees=25, twistLimitDegrees=1}
]
