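# Physics asset for the face accessory of the male medium FacetBlom player cosmetic, defining a jointed rag chain alongside body collision shapes.
# Six dyn_rag bodies link in series from MSTR_rag, with tight twist, narrow swing1 and wider swing2 limits that shrink toward the tip.
# Unconstrained capsules cover the collar, sword, pelvis, upper spine, clavicles and arms, giving collision volumes around the rag.
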PhysicsAsset /BRCosmetics/Characters/Player/Male/Medium/Bodies/M_MED_FacetBlom/Meshes/Parts/M_MED_FacetBlom_FaceAcc_Physics
bodies {
 MSTR_collar_c [{
   shapeType=Capsule
   length=7.5695
   radius=4.05
   shapeName=MSTR_collar_c_capsule
   center=<0.6732 6.0461 -12.0458>
   rotation=<rot 0 0.0001 -70.7167>
  }]
 MSTR_rag [{shapeType=Capsule, length=0.505, radius=0.505, shapeName=MSTR_rag_capsule, rotation=<rot 0 -90 0>}]
 MSTR_sword [{
   shapeType=Capsule
   length=41.757
   radius=10.4147
   shapeName=MSTR_sword_capsule
   center=<1.1509 3.0095 0.2834>
   rotation=<rot -2.0688 -100.261 78.7226>
  }]
 clavicle_l [{shapeType=Capsule, length=16.2284, radius=11.2672, shapeName=clavicle_l_capsule, center=<10.4406 0 0.9975>, rotation=<rot 89 0 0>}]
 clavicle_r [{shapeType=Capsule, length=16.2284, radius=11.2672, shapeName=clavicle_r_capsule, center=<-10.4406 0 -0.9975>, rotation=<rot 89 0 -180>}]
 dyn_rag_01 [
  {shapeType=Capsule, length=8.6414, radius=1.6959, shapeName=dyn_rag_01_capsule, center=<8.4461 0.4243 -1.6106>, rotation=<rot 90 -88.9749 0>}
  {
   shapeType=Capsule
   length=6.8727
   radius=2.3411
   shapeName=dyn_rag_01_capsule
   center=<12.3616 -0.0175 -1.8771>
   rotation=<rot 10.1612 0.0957 -89.4573>
  }
 ]
 dyn_rag_02 [
  {
   shapeType=Capsule
   length=5.7406
   radius=4.3565
   shapeName=dyn_rag_02_capsule
   center=<3.3123 -0.1139 -3.2869>
   rotation=<rot 0 -1.3206 -87.4886>
  }
  {
   shapeType=Capsule
   length=8.2712
   radius=3.8992
   shapeName=dyn_rag_02_capsule
   center=<10.4499 0.0327 -2.9883>
   rotation=<rot 6.0099 -0.0661 -90.6313>
  }
 ]
 dyn_rag_03 [{shapeType=Capsule, length=10.0596, radius=5.1532, shapeName=dyn_rag_03_capsule, center=<5.3982 0 -3.893>, rotation=<rot 0 0 90>}]
 dyn_rag_04 [{
   shapeType=Capsule
   length=8.0488
   radius=10.7384
   shapeName=dyn_rag_04_capsule
   center=<7.9486 -1.9426 -7.4531>
   rotation=<rot 0 -0.509 -87.0514>
  }]
 dyn_rag_05 [{shapeType=Capsule, length=8.8765, radius=6.3921, shapeName=dyn_rag_05_capsule, center=<9.7745 -5.8774 -4.4732>, rotation=<rot 0 0 90>}]
 dyn_rag_06 [{
   shapeType=Capsule
   length=15.2052
   radius=2.8686
   shapeName=dyn_rag_06_capsule
   center=<11.8311 -10.9729 0>
   rotation=<rot 0 73.479 -88.7665>
  }]
 lowerarm_l [{shapeType=Capsule, length=18.25, radius=6.6023, shapeName=lowerarm_l_capsule, center=<8.0201 0 0>, rotation=<rot 90 0 0>}]
 lowerarm_r [{shapeType=Capsule, length=19.937, radius=9.2578, shapeName=lowerarm_r_capsule, center=<-8.0201 0 0>, rotation=<rot 90 180 0>}]
 pelvis [{shapeType=Capsule, length=35, radius=15, shapeName=pelvis_capsule, center=<7.7829 -7.4576 -0.2602>}]
 spine_03 [{shapeType=Capsule, length=35, radius=15, shapeName=spine_03_capsule, center=<0 -4.7957 0>}]
 spine_05 [
  {shapeType=Capsule, length=30, radius=15, shapeName=spine_05_capsule, center=<0 -4.0806 0>}
  {shapeType=Capsule, length=15.8806, radius=14.1194, shapeName=spine_05_capsule, center=<6.2117 -6.8404 0>, rotation=<rot 0 5.2032 0>}
 ]
 upperarm_l [{shapeType=Capsule, length=19.9972, radius=8.7012, shapeName=upperarm_l_capsule, center=<7.8495 0 0>, rotation=<rot 90 0 0>}]
 upperarm_r [{shapeType=Capsule, length=19.9972, radius=8.7012, shapeName=upperarm_r_capsule, center=<-7.8495 0 0>, rotation=<rot 90 180 0>}]
}
constraints [
 {bone1Name=dyn_rag_01, bone2Name=MSTR_rag, swing1LimitDegrees=25, swing2LimitDegrees=80, twistLimitDegrees=0}
 {bone1Name=dyn_rag_02, bone2Name=dyn_rag_01, swing1LimitDegrees=20, swing2LimitDegrees=80, twistLimitDegrees=3}
 {bone1Name=dyn_rag_03, bone2Name=dyn_rag_02, swing1LimitDegrees=20, swing2LimitDegrees=80, twistLimitDegrees=3}
 {bone1Name=dyn_rag_04, bone2Name=dyn_rag_03, swing1LimitDegrees=20, swing2LimitDegrees=65, twistLimitDegrees=3}
 {bone1Name=dyn_rag_05, bone2Name=dyn_rag_04, swing1LimitDegrees=20, swing2LimitDegrees=50, twistLimitDegrees=5}
 {bone1Name=dyn_rag_06, bone2Name=dyn_rag_05, swing1LimitDegrees=25, swing2LimitDegrees=50, twistLimitDegrees=3}
]
