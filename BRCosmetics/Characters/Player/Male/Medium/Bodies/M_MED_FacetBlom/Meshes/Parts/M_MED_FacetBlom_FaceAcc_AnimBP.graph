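# Post-copy AnimBP for the FacetBlom character's face accessory, animating its rag bones with secondary motion.
# Copied parent and body poses pass through a Control Rig and a scaled head counter-rotation into a cached pose.
# Without a backpack or cape it uses head-space RigidBody physics; the matching tagged backpack swaps in another physics asset.
# Other back items use head/spine constraints with AnimDynamics on the tip, plus additive bone offsets.
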AnimBlueprint /BRCosmetics/Characters/Player/Male/Medium/Bodies/M_MED_FacetBlom/Meshes/Parts/M_MED_FacetBlom_FaceAcc_AnimBP
target_skeleton=&/M_MED_FacetBlom_FaceAcc_Skeleton
graphs [
 AnimGraph {nodes [
   n1 = "Constraint - dyn_rag_01 (PP)"!(ComponentPose=Local To Component(LocalPose=Use cached pose 'PreDynamics'()), Alpha=1, ConstraintWeights_0=0.6, ConstraintWeights_1:float=0.4) -> (Pose) {
    boneToModify {boneName=dyn_rag_01}
    constraintSetup [{targetBone {boneName=head}}, {targetBone {boneName=spine_05}}]
    constraintWeights [1,1]
   }
   n2 = "Constraint - dyn_rag_02 (PP)"!(ComponentPose=@n1, Alpha=1, ConstraintWeights_0=0.6, ConstraintWeights_1:float=0.4) -> (Pose) {
    boneToModify {boneName=dyn_rag_02}
    constraintSetup [{targetBone {boneName=head}}, {targetBone {boneName=spine_05}}]
    constraintWeights [1,1]
   }
   n3 = "Constraint - dyn_rag_03 (P)"(ComponentPose=@n2, Alpha=1, ConstraintWeights_0=1) {boneToModify {boneName=dyn_rag_03}, constraintSetup [{targetBone {boneName=spine_05}}], constraintWeights [1]}
   n4 = "Constraint - dyn_rag_04 (PP)"!(ComponentPose=@n3, Alpha=1, ConstraintWeights_0=0.5, ConstraintWeights_1:float=0.5) -> (Pose) {
    boneToModify {boneName=dyn_rag_04}
    constraintSetup [{targetBone {boneName=spine_04}}, {targetBone {boneName=spine_03}}]
    constraintWeights [1,1]
   }
   n5 = "Transform (Modify) Bone - Bone: dyn_rag_02"(ComponentPose=@n4, Alpha=1, Translation="-3,0,-3", Rotation="0,0,0", Scale="1,1,1") {
    rotationMode=BMM_Additive
    rotationSpace=BCS_BoneSpace
    translationMode=BMM_Additive
    translationSpace=BCS_BoneSpace
    boneToModify {boneName=dyn_rag_02}
    translation {x=-3, z=-3}
   }
   n6 = "Transform (Modify) Bone - Bone: dyn_rag_03"(ComponentPose=@n5, Alpha=1, Translation="0,0,2", Rotation="10,0,0", Scale="1,1,1") {
    rotationMode=BMM_Additive
    rotationSpace=BCS_BoneSpace
    translationMode=BMM_Additive
    translationSpace=BCS_BoneSpace
    boneToModify {boneName=dyn_rag_03}
    rotation {pitch=10}
    translation {z=2}
   }
   n7 = "Transform (Modify) Bone - Bone: dyn_rag_05"(ComponentPose=@n6, Alpha=1, Translation="0,0,-1", Rotation="-15,0,0", Scale="1,1,1") {
    rotationMode=BMM_Additive
    rotationSpace=BCS_BoneSpace
    translationMode=BMM_Additive
    translationSpace=BCS_BoneSpace
    boneToModify {boneName=dyn_rag_05}
    rotation {pitch=-15}
    translation {z=-1}
   }
   n8 = Anim Dynamics - Bone: dyn_rag_06(ComponentPose=@n7, Alpha=1, ExternalForce="0,0,0") {
    angularDampingOverride=0.9999
    angularSpringConstant=2500
    bAngularSpring=true
    bOverrideAngularDamping=true
    bUseGravityOverride=true
    gravityScale=0.35
    simulationSpace=RootRelative
    boundBone {boneName=dyn_rag_06}
    componentLinearAccScale=<1 1 1>
    componentLinearVelScale=<1 1 1>
    gravityOverride {z=-980}
    physicsBodyDefinitions [{
      boundBone {boneName=dyn_rag_06}
      boxExtents=<20 20 20>
      constraintSetup {angularLimitsMax=<15 15 15>, angularLimitsMin=<-15 -15 -15>, angularTarget=<1 0 0>}
      localJointOffset=<5 0 0>
     }]
   }
   n9 = "Transform (Modify) Bone - Bone: dyn_rag_02"(ComponentPose=@n8, Alpha=1, Translation="0,0,0", Rotation="5,0,0", Scale="1,1,1") {
    rotationMode=BMM_Additive
    rotationSpace=BCS_BoneSpace
    translationMode=BMM_Additive
    translationSpace=BCS_BoneSpace
    boneToModify {boneName=dyn_rag_02}
    rotation {pitch=5}
   }
   n10 = "Transform (Modify) Bone - Bone: dyn_rag_03"(ComponentPose=@n9, Alpha=1, Translation="0,0,0", Rotation="2,0,0", Scale="1,1,1") {
    rotationMode=BMM_Additive
    rotationSpace=BCS_BoneSpace
    translationMode=BMM_Additive
    translationSpace=BCS_BoneSpace
    boneToModify {boneName=dyn_rag_03}
    rotation {pitch=2}
   }
   n11 = "Transform (Modify) Bone - Bone: dyn_rag_05"(ComponentPose=@n10, Alpha=1, Translation="0,0,0", Rotation="2,0,0", Scale="1,1,1") {
    rotationMode=BMM_Additive
    rotationSpace=BCS_BoneSpace
    translationMode=BMM_Additive
    translationSpace=BCS_BoneSpace
    boneToModify {boneName=dyn_rag_05}
    rotation {pitch=2}
   }
   n12 = RigidBody!(ComponentPose=Local To Component(LocalPose=Use cached pose 'PreDynamics'()), Alpha, OverrideWorldGravity:Vector="0,0,-980", ExternalForce) -> (Pose) {
    bOverrideWorldGravity=true
    overridePhysicsAsset=&/M_MED_FacetBlom_FaceAcc_BackpackEquip_Physics
    simulationSpace=BaseBoneSpace
    baseBoneRef {boneName=head}
    componentAppliedLinearAccClamp=<3000 3000 3000>
    overrideWorldGravity {z=-980}
    simSpaceSettings {
     maxAngularAcceleration=40
     maxAngularVelocity=8
     maxLinearAcceleration=3200
     maxLinearVelocity=800
     worldAlpha=1
     externalLinearDragV=<0.8 0.75 0.75>
    }
   }
   n13 = "Transform (Modify) Bone - Bone: MSTR_rag"(ComponentPose=@n12, Alpha=1, Translation="0,0,0", Rotation="-18.536,-3.061,10.488", Scale="1,1,1") {
    rotationMode=BMM_Additive
    rotationSpace=BCS_BoneSpace
    translationMode=BMM_Additive
    translationSpace=BCS_BoneSpace
    boneToModify {boneName=MSTR_rag}
    rotation=<rot -18.536 -3.061 10.488>
   }
   n14 = "Transform (Modify) Bone - Bone: MSTR_rag"(ComponentPose=@n13, Alpha=1, Translation="0,0,0", Rotation="-18.536,-3.061,10.488", Scale="1,1,1") {
    rotationMode=BMM_Additive
    rotationSpace=BCS_BoneSpace
    translationMode=BMM_Additive
    translationSpace=BCS_BoneSpace
    boneToModify {boneName=MSTR_rag}
    rotation=<rot -18.536 -3.061 10.488>
   }
   n15 = "Transform (Modify) Bone - Bone: MSTR_rag"(ComponentPose=@n14, Alpha=1, Translation="0,0,0", Rotation="-22.087,-8.145,3.094", Scale="1,1,1") {
    rotationMode=BMM_Additive
    rotationSpace=BCS_BoneSpace
    translationMode=BMM_Additive
    translationSpace=BCS_BoneSpace
    boneToModify {boneName=MSTR_rag}
    rotation=<rot -22.087 -8.145 3.094>
   }
   n16 = "Transform (Modify) Bone - Bone: MSTR_rag"(ComponentPose=@n15, Alpha=1, Translation="0,0,0", Rotation="-3,0,0", Scale="1,1,1") {
    rotationMode=BMM_Additive
    rotationSpace=BCS_BoneSpace
    translationMode=BMM_Additive
    translationSpace=BCS_BoneSpace
    boneToModify {boneName=MSTR_rag}
    rotation {pitch=-3}
   }
   n17 = "Transform (Modify) Bone - Bone: MSTR_rag"(ComponentPose=@n16, Alpha=1, Translation="0,0,0", Rotation="-15,0,0", Scale="1,1,1") {
    rotationMode=BMM_Additive
    rotationSpace=BCS_BoneSpace
    translationMode=BMM_Additive
    translationSpace=BCS_BoneSpace
    alphaScaleBiasClamp {bInterpResult=true, interpSpeedDecreasing=5, interpSpeedIncreasing=5}
    boneToModify {boneName=MSTR_rag}
    rotation {pitch=-15}
   }
   n18 = "Transform (Modify) Bone - Bone: dyn_rag_03"(ComponentPose=@n17, Alpha=1, Translation="0,0,0", Rotation="3,0,0", Scale="1,1,1") {
    rotationMode=BMM_Additive
    rotationSpace=BCS_BoneSpace
    translationMode=BMM_Additive
    translationSpace=BCS_BoneSpace
    boneToModify {boneName=dyn_rag_03}
    rotation {pitch=3}
   }
   n19 = "Transform (Modify) Bone - Bone: MSTR_rag"(ComponentPose=@n18, Alpha=1, Translation="0,0,0", Rotation="-15,0,0", Scale="1,1,1") {
    rotationMode=BMM_Additive
    rotationSpace=BCS_BoneSpace
    translationMode=BMM_Additive
    translationSpace=BCS_BoneSpace
    alphaScaleBiasClamp {bInterpResult=true, interpSpeedDecreasing=5, interpSpeedIncreasing=5}
    boneToModify {boneName=MSTR_rag}
    rotation {pitch=-15}
   }
   n20 = "Transform (Modify) Bone - Bone: MSTR_rag"(ComponentPose=@n19, Alpha=1, Translation="0,0,0", Rotation="-10,0,0", Scale="1,1,1") {
    rotationMode=BMM_Additive
    rotationSpace=BCS_BoneSpace
    translationMode=BMM_Additive
    translationSpace=BCS_BoneSpace
    alphaScaleBiasClamp {bInterpResult=true, interpSpeedDecreasing=7, interpSpeedIncreasing=0}
    boneToModify {boneName=MSTR_rag}
    rotation {pitch=-10}
   }
   n21 = "Transform (Modify) Bone - Bone: MSTR_rag"(ComponentPose=Local To Component(LocalPose=Use cached pose 'PreDynamics'()), Alpha=1, Translation="0,0,0", Rotation="-15,0,0", Scale="1,1,1") {
    rotationMode=BMM_Additive
    rotationSpace=BCS_BoneSpace
    translationMode=BMM_Additive
    translationSpace=BCS_BoneSpace
    boneToModify {boneName=MSTR_rag}
    rotation {pitch=-15}
   }
   n22 = RigidBody(ComponentPose=@n21, OverrideWorldGravity="0,0,-980") {
    bOverrideWorldGravity=true
    overridePhysicsAsset=&/M_MED_FacetBlom_FaceAcc_Physics
    simulationSpace=BaseBoneSpace
    baseBoneRef {boneName=head}
    componentAppliedLinearAccClamp=<3000 3000 3000>
    overrideWorldGravity {z=-980}
    simSpaceSettings {
     maxAngularAcceleration=40
     maxAngularVelocity=8
     maxLinearAcceleration=3200
     maxLinearVelocity=800
     worldAlpha=1
     externalLinearDragV=<0.8 0.75 0.75>
    }
   }
   n23 = "Transform (Modify) Bone - Bone: dyn_rag_03"(ComponentPose=@n22, Alpha=1, Translation="0,0,0", Rotation="3,0,0", Scale="1,1,1") {
    rotationMode=BMM_Additive
    rotationSpace=BCS_BoneSpace
    translationMode=BMM_Additive
    translationSpace=BCS_BoneSpace
    boneToModify {boneName=dyn_rag_03}
    rotation {pitch=3}
   }
   n24 = "Transform (Modify) Bone - Bone: MSTR_rag"(ComponentPose=@n23, Alpha=1, Translation="0,0,0", Rotation="-15,0,0", Scale="1,1,1") {
    rotationMode=BMM_Additive
    rotationSpace=BCS_BoneSpace
    translationMode=BMM_Additive
    translationSpace=BCS_BoneSpace
    alphaScaleBiasClamp {bInterpResult=true, interpSpeedDecreasing=5, interpSpeedIncreasing=5}
    boneToModify {boneName=MSTR_rag}
    rotation {pitch=-15}
   }
   n25 = "Transform (Modify) Bone - Bone: MSTR_rag"(ComponentPose=@n24, Alpha=1, Translation="0,0,0", Rotation="-10,0,0", Scale="1,1,1") {
    rotationMode=BMM_Additive
    rotationSpace=BCS_BoneSpace
    translationMode=BMM_Additive
    translationSpace=BCS_BoneSpace
    alphaScaleBiasClamp {bInterpResult=true, interpSpeedDecreasing=7, interpSpeedIncreasing=0}
    boneToModify {boneName=MSTR_rag}
    rotation {pitch=-10}
   }
   n26 = Blend Poses by bool(
    BlendPose_0=Blend Poses by bool(
     BlendPose_0=Component To Local(ComponentPose=@n20)
     BlendPose_1=Component To Local(ComponentPose=@n11)
     bActiveValue=Get OwnBackpackEquiped() -> (OwnBackpackEquiped:bool)
    ) {blendType=Linear}
    BlendPose_1=Component To Local(ComponentPose=@n25)
    bActiveValue=Get IsWearingBackpackOrCape() -> (IsWearingBackpackOrCape:bool)
   ) {blendType=Linear}
   Output Pose(Result=Component To Local(ComponentPose=Local To Component(LocalPose=@n26))) {layerGroup=None}
   n27 = Layered blend per bone(
    BasePose=Copy Pose From Mesh() {bCopyCurves=true, bUseAttachedParent=true}
    BlendPoses_0=Copy Pose From Mesh(SourceMeshComponent=Get Body_Mesh()) {bUseAttachedParent=true}
   ) {layerSetup [{branchFilters [
       {boneName=clavicle_l}
       {boneName=clavicle_r}
       {boneName=upperarm_l}
       {boneName=upperarm_r}
       {boneName=lowerarm_l}
       {boneName=lowerarm_r}
       {boneName=MSTR_collar_c}
       {boneName=dyn_collar_c}
       {boneName=None}
      ]}]}
   Save cached pose 'StartCache'(Pose=@n27) {cacheName=StartCache}
   n28 = Control Rig!(Source=Use cached pose 'StartCache'(), Alpha, skyDive:bool=Get bIsSkydiveFloating(), OwnBackpackEquipped:bool=Get OwnBackpackEquiped()) -> (Pose) {controlRigAssetReference {blueprintRigClass=&/M_MED_FacetBlom_FaceAcc_CtrlRig.M_MED_FacetBlom_FaceAcc_CtrlRig_C}}
   n29 = Local To Component(LocalPose=@n28)
   n30 = Get head rot() -> (head rot:Rotator)
   Save cached pose 'PreDynamics'(Pose=Component To Local(ComponentPose="Transform (Modify) Bone - Bone: head"(ComponentPose=@n29, Alpha=1, Translation="0,0,0", Rotation=@n30, Scale="1,1,1") {rotationMode=BMM_Additive, rotationSpace=BCS_BoneSpace, boneToModify {boneName=head}})) {cacheName=PreDynamics}
  ]}
 EventGraph {nodes [
   Event BlueprintUpdateAnimation()
   > n1 = Sequence!(execute) -> (then_0, then_1, then_2:Exec)
   n2 = "Gravity Override State Function (Parameter Asset) (Nativized)"(
    execute=@n1.then_1
    CharacterPartAnimInstance=Self-Reference()
    GameTimeInSeconds=GetGameTimeInSeconds()
    Parameters=&/M_MED_FacetBlom_Parameter
   )
   > Set GravityOverride(GravityOverride=@n2)
   n3 = GetOwnerSkeletalMeshComponent()
   n4 = BreakRotator(InRot=BreakTransform(InTransform=GetDeltaTransformFromRefPose(self=@n3, BoneName=head, BaseName=neck_02)).Rotation)
   n5 = Get Y percentage() -> (Y percentage:float)
   n6 = Get X percentage() -> (X percentage:float)
   n7 = Get Z percentage() -> (Z percentage:float)
   Set head rot(
    execute=@n1.then_2
    head rot:Rotator=MakeRotator(
     Roll=float * float!(A=@n4, B=-1, C:double=@n6) -> (ReturnValue)
     Pitch=float * float(A=@n4.Pitch, B=-1, C=@n5)
     Yaw=float * float(A=@n4.Yaw, B=-1, C=@n7)
    )
   ) -> (Output_Get:Rotator)
   n8 = CheckForBackbling()
   Set IsWearingBackpackOrCape(execute=@n1.then_0, IsWearingBackpackOrCape:bool=OR Boolean!(A=@n8, B=@n8.bIsBackblingACape, C:bool=@n8.bIsBackblingLong) -> (ReturnValue)) -> (Output_Get:bool)
   > Set IsWearingCape(IsWearingCape:bool=@n8.bIsBackblingACape) -> (Output_Get:bool)
   n9 = Event OnCharacterCustomizationLoaded()
   > Set Body_Mesh(Body_Mesh=Get Array Item with EFort Custom Part Type(Array=@n9.CustomMeshParts))
   > Get Dynamic Material Instance(SkeletalMeshComponent=Get Array Item with EFort Custom Part Type(Array=@n9.CustomMeshParts, PartType=Face))
   > Set OwnBackpackEquiped(OwnBackpackEquiped:bool=BackpackItemDefinitionHasAnyMetaTag!(Tags:Array of Gameplay Tag Structures=Get Tags() -> (Tags:Array of Gameplay Tag Structures), bRequiresEquipped:bool=false) -> (ReturnValue:bool)) -> (Output_Get:bool)
  ]}
]
variables [
 Body_Mesh {type=SkeletalMeshComponent}
 GravityOverride {type=Vector}
 head rot {type=Rotator}
 IsWearingBackpackOrCape {type=boolean}
 IsWearingCape {type=boolean}
 OwnBackpackEquiped {type=boolean}
 Tags {type="array<GameplayTag>", default [{tagName=Cosmetics.MetaTags.FacetBlom}]}
 X Percentage {type=number}
 Y percentage {type=number}
 Z percentage {type=number, default=1}
]
